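; Looping gameplay cue for the repair tool that plays a surface-dependent loop sound at the point being repaired.
; Application moves the audio component, sets a short self-removal timer and fades in a sound chosen by GetSoundFromSurface from the target's wood, stone or metal tag, treating vehicles as metal and defaulting to generic.
; Timer expiry or cue removal fades the loop out, waits, then removes the cue from the target.
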(event EventOnApplication(EitherVFX-preferthisone)
  (params (MyTarget :type Actor%20Object%20Reference) (Parameters :type Gameplay%20Cue%20Parameters%20Structure%20%28by%20ref%29) (ParticleComponents :type Array%20of%20FXSystem%20Component%20Object%20References) (AudioComponents :type Array%20of%20Audio%20Component%20Object%20References) (BurstCameraShakeInstance :type Camera%20Shake%20Base%20Object%20Reference) (BurstDecalInstance :type Decal%20Actor%20Object%20Reference))
  (bind (_normalizedmagnitude _rawmagnitude _effectcontext _matchedtagname _originaltag _aggregatedsourcetags _aggregatedtargettags _location) (Ability|GameplayCue|BreakGameplayCueParameters Parameters))
  (bind _sfx_surfaceloop (Variables|Default|GetSFX_SurfaceLoop))
  ;[0.583 0.2655 1] Set up looping surface sound and get sound to play based on surface
  (Transformation|SetWorldLocation (Variables|Default|GetSFX_SurfaceLoop) _location false false)
  (bind _returnvalue (Utilities|Time|SetTimerbyEvent (delegate Custom|DestroyCue) 0.3))
  (Variables|Default|SetRemovalTimer _returnvalue)
  (if (== MyTarget (Variables|Default|GetTargetActor))
    (Utilities|FlowControl|DoOnce false
      (:Completed
        _))
    (else
      _))
  (Variables|Default|SetTargetActor MyTarget)
  (Audio|Components|Audio|SetSound _sfx_surfaceloop (CallFunction|GetSoundfromSurface))
  (Audio|Components|Audio|FadeIn _sfx_surfaceloop 0.3 1.0 0.0 "Linear" :!class /Script/Engine.AudioComponent))

(event Custom|DestroyCue
  (bind _sfx_surfaceloop (Variables|Default|GetSFX_SurfaceLoop))
  (bind _fade_out_duration (Variables|Default|GetFade%20Out%20Duration))
  ;[0.583 0.2655 1] Set up looping surface sound and get sound to play based on surface
  (Audio|Components|Audio|FadeOut (Variables|Default|GetSFX_SurfaceLoop) _fade_out_duration 0.0 "Linear" :!class /Script/Engine.AudioComponent)
  (Utilities|FlowControl|Delay 0 _fade_out_duration)
  (Utilities|FlowControl|Sequence
    (:then_0
      (Utilities|FlowControl|DoOnce false
        (:Completed
          (Variables|Default|SetTargetActor (stale EventOnApplication(EitherVFX-preferthisone) MyTarget))
          (Audio|Components|Audio|SetSound _sfx_surfaceloop (CallFunction|GetSoundfromSurface))
          (Audio|Components|Audio|FadeIn _sfx_surfaceloop 0.3 1.0 0.0 "Linear" :!class /Script/Engine.AudioComponent))))
    (:then_1
      (GameplayCue|RemoveGameplayCue (Variables|Default|GetTargetActor) (Variables|Default|GetGameplay%20Cue%20Tag)))))

(event EventOnRemoval(EitherVFX-preferthisversion)
  (params (MyTarget :type Actor%20Object%20Reference) (Parameters :type Gameplay%20Cue%20Parameters%20Structure%20%28by%20ref%29) (ParticleComponents :type Array%20of%20FXSystem%20Component%20Object%20References) (AudioComponents :type Array%20of%20Audio%20Component%20Object%20References) (BurstCameraShakeInstance :type Camera%20Shake%20Base%20Object%20Reference) (BurstDecalInstance :type Decal%20Actor%20Object%20Reference))
  (bind _sfx_surfaceloop (Variables|Default|GetSFX_SurfaceLoop))
  (bind _fade_out_duration (Variables|Default|GetFade%20Out%20Duration))
  ;[0.583 0.2655 1] Set up looping surface sound and get sound to play based on surface
  (Audio|Components|Audio|FadeOut (Variables|Default|GetSFX_SurfaceLoop) _fade_out_duration 0.0 "Linear" :!class /Script/Engine.AudioComponent)
  (Utilities|FlowControl|Delay 0 _fade_out_duration)
  (Utilities|FlowControl|Sequence
    (:then_0
      (Utilities|FlowControl|DoOnce false
        (:Completed
          (Variables|Default|SetTargetActor (stale EventOnApplication(EitherVFX-preferthisone) MyTarget))
          (Audio|Components|Audio|SetSound _sfx_surfaceloop (CallFunction|GetSoundfromSurface))
          (Audio|Components|Audio|FadeIn _sfx_surfaceloop 0.3 1.0 0.0 "Linear" :!class /Script/Engine.AudioComponent))))
    (:then_1
      (GameplayCue|RemoveGameplayCue (Variables|Default|GetTargetActor) (Variables|Default|GetGameplay%20Cue%20Tag)))))

(fn GetSoundFromSurface
  (bind _targetactor (Variables|Default|GetTargetActor))
  (bind _returnvalue (GameplayTags|HasMatchingGameplayTag(Message) _targetactor "(TagName=\"Building.Resource.Metal\")"))
  (bind _returnvalue_2 (GameplayTags|HasMatchingGameplayTag(Message) _targetactor "(TagName=\"Building.Resource.Stone\")"))
  (bind _returnvalue_3 (GameplayTags|HasMatchingGameplayTag(Message) _targetactor "(TagName=\"Building.Resource.Wood\")"))
  (return (select _returnvalue_3 (Variables|SoundTypes|GetWoodSurfaceSound) (select _returnvalue_2 (Variables|SoundTypes|GetGenericSurfaceSound) (select _returnvalue (Variables|SoundTypes|GetMetalSurfaceSound) (select (Utilities|DoesObjectImplementInterface _targetactor "/Script/FortniteGame.FortVehicleInterface") (Variables|SoundTypes|GetMetalSurfaceSound) (Variables|SoundTypes|GetGenericSurfaceSound)))))))

(fn ConstructionScript)

(fn OnBecomeRelevant
  (params (MyTarget :type Actor%20Object%20Reference) (Parameters :type Gameplay%20Cue%20Parameters%20Structure%20%28by%20ref%29))
  (bind _sfx_surfaceloop (Variables|Default|GetSFX_SurfaceLoop))
  (if (not (Components|Activation|IsActive _sfx_surfaceloop))
    (Components|Activation|Activate _sfx_surfaceloop false)
    (return false)))
